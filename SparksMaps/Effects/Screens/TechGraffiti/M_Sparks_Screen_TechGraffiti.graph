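# Unlit material for Sparks stage screens showing animated tech-graffiti art in stage colours that respond to music beats.
# The texture is flippable per primitive, scaled, rotated and panned; its channels supply shape, sine-pulse phase and colour masks.
# Pulsing art plus a vertical gradient glow feeds the screen-details function as emissive.
# Named reroutes carry primary and secondary stage colours and a beat-driven reactivity value from an audio parameter collection through a curve atlas.
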Material /SparksMaps/Effects/Screens/TechGraffiti/M_Sparks_Screen_TechGraffiti
graph {
 properties {
  bAutomaticallySetUsageInEditor=false
  bHasPixelAnimation=true
  bUsedWithNanite=true
  pixelDepthOffsetMode=PDOM_Legacy
  shadingModel=MSM_Unlit
  usedShadingModels=Unlit
 }
 nodes [
  n1 = NamedRerouteUsage()
  n2 = NamedRerouteUsage()
  n3 = MaterialFunctionCall() {materialFunction=&LinearGradient}
  n4 = Constant() {r=5}
  n5 = Multiply(A=@n2, B=Multiply(A=Power(Base=@n3.VGradient, Exp=@n4), B=Constant() {r=0.25}))
  n6 = TextureCoordinate()
  n7 = ComponentMask(None=@n6) {r=true}
  n8 = LinearInterpolate(
   A=@n6
   B=AppendVector(A=OneMinus(None=@n7), B=ComponentMask(None=@n6) {g=true})
   Alpha=ScalarParameter() {bUseCustomPrimitiveData=true, parameterName=Flip, primitiveDataIndex=4}
  )
  n9 = Multiply(A=ScalarParameter() {defaultValue=1.5, parameterName=TextureScale}, B=LinearInterpolate(Alpha=NamedRerouteUsage()) {constA=1, constB=1.01})
  n10 = Panner(Coordinate=Rotator(
    Coordinate=MaterialFunctionCall(UVs=@n8, Texture Scale=@n9) {materialFunction=&ScaleUVsByCenter}.UVs
    Time=ScalarParameter() {defaultValue=3, parameterName=TextureRotation}
   )) {speedY=0.025}
  n11 = TextureSample(UVs=@n10) {samplerType=SAMPLERTYPE_LinearColor, texture=&/T_Sparks_Screen_TechGraffiti}
  n12 = NamedRerouteUsage()
  n13 = Step(Y=@n11.B) {constX=0.5}
  n14 = Multiply(A=Multiply(A=LinearInterpolate(A=@n2, B=@n12, Alpha=@n13), B=Constant() {r=3}) {constB=5}, B=@n11.R)
  n15 = TextureSample(UVs=@n10) {samplerType=SAMPLERTYPE_LinearColor, texture=&T_Sparks_Screen_TechGraffiti}
  n16 = Time()
  n17 = Constant() {r=0.5}
  n18 = ComponentMask(None=Add(A=@n15.G, B=Multiply(A=@n16, B=@n17) {constB=0.2})) {g=true}
  n19 = SmoothStep(Min=Constant() {r=0.5}, Max=Constant() {r=1}, Value=ConstantBiasScale(None=Sine(None=@n18))) {constMin=0.9}
  n20 = Add(A=Power(Base=Multiply(A=@n19) {constB=2}))
  n21 = Add(
   A=Multiply(A=Multiply(A=@n2) {constB=0.01}, B=Step(Y=@n11.B) {constX=0.1, constY=0.1})
   B=Multiply(A=Multiply(A=@n14, B=@n20), B=Constant() {r=0.125}) {constB=0.25}
  )
  MaterialFunctionCall!(Emissive=Multiply(A=LinearInterpolate(Alpha=@n1) {constA=0.25, constB=1.1}, B=Add(A=@n5, B=@n21)), Scanlines Tile) {materialFunction=/SparksMaps/Effects/Screens/Materials/MF/MF_Sparks_Screen_Details}
  Multiply(A=TextureCoordinate(), B=Multiply(A=@n15.G, B=Constant() {r=1}) {constB=0.3})
  n22 = MaterialFunctionCall!(CustomUPosition) {materialFunction=/SparksStageShow/Systems/StageManager/FXManager/MaterialFunctions/MF_MojoStageColors}
  NamedRerouteDeclaration(None=@n22.Primary) {name=Primary Color, nodeColor=<rgba 0 0 0 1>}
  n23 = CurveAtlasRowParameter(CurveTime=CollectionParameter() {collection=/FMAudioReactiveData/AudioAnalysis/MPC_Sparks_Audio_Parameters, parameterName=BeatLifetime}) {
   atlas=/SparksMaps/Effects/Screens/Curves/CurveAtlas_Audio_Beat_A
   curve=/SparksMaps/Effects/Screens/Curves/Curve_Audio_Beat_Remap_001
   parameterName=ReactivityCurve
  }
  NamedRerouteDeclaration(None=MaterialFunctionCall(Input=@n23, Input Low=Constant(), Input High=Constant() {r=1}, Target Low=Constant() {r=0.75}, Target High=Constant() {r=2}) {materialFunction=&RemapValueRange}.Result) {name=Reactivity, nodeColor=<rgba 0 0 0 1>}
  NamedRerouteDeclaration(None=@n22.Secondary) {name=Secondary Color, nodeColor=<rgba 0 0 0 1>}
  n24 = Time()
  Reroute(None=Frac(None=Multiply(A=@n24) {constB=2}))
  VectorParameter() {group=ScreenThemeParams, parameterName=TestFlameColor, defaultValue=<rgba 1 0.0737 0.025 1>}
 ]
}
